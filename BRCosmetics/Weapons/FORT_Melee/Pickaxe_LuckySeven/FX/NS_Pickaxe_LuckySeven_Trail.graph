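# Niagara weapon trail for the Lucky Seven pickaxe, drawing short-lived animation-driven ribbons behind the melee weapon.
# A single FN_AnimTrail emitter spawns three trails between socket index pairs and feeds two ribbon renderers.
# Spawn-group masks give the trails different widths, red or black colors and alpha fade timing.
# Curves darken and fade them over lifetime, then a system time-of-day multiplier scales the final color.
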NiagaraSystem /BRCosmetics/Weapons/FORT_Melee/Pickaxe_LuckySeven/FX/NS_Pickaxe_LuckySeven_Trail
dependencies {
 usedDataInterfaces [&NiagaraDataInterfaceColorCurve, &NiagaraDataInterfaceCurve, &NiagaraDataInterfaceSocketReader]
 usedDynamicInputs []
 usedModules [
  &self:SystemSpawnScript.SystemScriptSource.SystemScriptGraph.NiagaraNodeAssignment_1.SetVariables_#1
  &SystemState
  &self:SystemSpawnScript.SystemScriptSource.SystemScriptGraph.NiagaraNodeAssignment_2.SetVariables_#2
 ]
}
emitters {FN_AnimTrail {
  emitterSpawnScript [&FN_RibbonIDOffsetUtility()]
  emitterUpdateScript [
   &EmitterState()
   &FN_SpawnAnimTrail(AnimTrailWidthMode=::FromSecond, Number of Anim Trails=3)
  ]
  particleSpawnScript [
   &FN_InitializeAnimTrailBoneIndecies(First Bone or Socket Index=1, Second Bone or Socket Index=0)
   &FN_InitializeAnimTrailBoneIndecies(AnimTrail Selection=1, First Bone or Socket Index=1, Second Bone or Socket Index=0) {module=FN_InitializeAnimTrailBoneIndecies001}
   &FN_InitializeAnimTrailBoneIndecies(AnimTrail Selection=2, First Bone or Socket Index=1, Second Bone or Socket Index=2) {module=FN_InitializeAnimTrailBoneIndecies002}
   &FN_InitializeAnimTrail(
    AnimTrailWidth=&MaskFloatBySpawnGroup(
     Masked Float=1.25
     Passthrough Float=&MaskFloatBySpawnGroup(Masked Float=0.1, Passthrough Float=2, Spawn Group Mask=1)
    )
    Color=&MaskLinearColorBySpawnGroup(
     Masked Linear Color=<rgba 1 0.0174 0.0537 0.75>
     Passthrough Linear Color=&MaskLinearColorBySpawnGroup(Masked Linear Color=<rgba 3 0.0783 0.1305 0.25>, Passthrough Linear Color=<rgba 0 0 0 1>, Spawn Group Mask=1)
    )
    Lifetime=0.325
   )
   &AddVelocity(Velocity=<0 400 0>) {enabled=false}
  ]
  particleUpdateScript [
   &SolveForcesAndVelocity(Clamp Velocity=false, Limit Acceleration=false)
   &FN_UpdateAnimTrail() {module=FN_UpdateAnimTrail001}
   &ScaleColor(
    Scale Alpha=&MaskFloatBySpawnGroup(
     Masked Float=&FloatFromCurve(FloatCurve=DataInterface() {propertyValues {Curve {keys [<at 0.175 1>,<at 0.3 0>]}}})
     Passthrough Float=&FloatFromCurve(FloatCurve=DataInterface() {propertyValues {Curve {keys [<at 0.175 1>,<at 0.225 0>]}}})
    )
    Scale RGB=&MakeVectorFromLinearColorRGB(Color=&ColorFromCurve(ColorCurve=DataInterface() {propertyValues {
        AlphaCurve {keys [<at 0 1 RCIM_Cubic>]}
        BlueCurve {keys [<at 0.1974 1 RCIM_Cubic>,<at 0.4705 0 RCIM_Cubic>]}
        GreenCurve {keys [<at 0.1974 1 RCIM_Cubic>,<at 0.4705 0 RCIM_Cubic>]}
        RedCurve {keys [<at 0.1974 1 RCIM_Cubic>,<at 0.4705 0 RCIM_Cubic>]}
       }}))
    ScaleA=true
    ScaleRGB=true
   )
   /SetVariables(Particles.Color=&MultiplyLinearColor(
     Linear Color=$Particles.Color
     Scale Factor=&MakeVector4_FromLinearColor(LinearColor=$System.TODMultiplier)
    )) {isSetParameters=true}
  ]
  rendererClasses [&NiagaraRibbonRendererProperties]
  renderers [
   {rendererClass=&NiagaraRibbonRendererProperties}
   {rendererClass=&NiagaraRibbonRendererProperties, rendererIndex=1}
  ]
 }}
summary {
 systemName=NS_Pickaxe_LuckySeven_Trail
 emitters [{emitterName=FN_AnimTrail, rendererClasses [&NiagaraRibbonRendererProperties]}]
 userVariables [
  User.bIsFrontEnd {type=NiagaraBool}
  User.bIsFrontEndPreview {type=NiagaraBool}
  User.PhysicalSurface {type=NiagaraInt32}
  User.RibbonIDOffset {type=NiagaraInt32}
 ]
}
